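# Niagara thruster effect for the Season 7 Battlestage, built as a single mesh emitter named Flash with two mesh renderers.
# One burst particle points down local -Z, loops its short lifetime without dying, and is sized by User.GlobalScale.
# A Timeline001-driven curve pulses mesh scale, multiplied by a ThrusterMask stretch whose setter (age, timeline, partition-signed RightDot) is disabled.
# The other listed user parameters are unreferenced in the graph shown.
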NiagaraSystem /SparksMaps/Environment/S7/Battlestage/Effects/NS_BattleStage_Thruster_S7
dependencies {
 usedDataInterfaces [
  &NiagaraDataInterfaceArrayFloat
  &NiagaraDataInterfaceMeshRendererInfo
  &NiagaraDataInterfaceCurve
 ]
 usedDynamicInputs [&SelectFloatFromArray]
 usedModules [&SystemState]
}
emitters {Flash {
  emitterUpdateScript [
   &EmitterState(
    Inactive Response=::NewEnumerator1
    Life Cycle Mode=::NewEnumerator1
    Loop Behavior=::NewEnumerator1
    Loop Duration Mode=::NewEnumerator1
    UseLoopDelay=true
   )
   &SpawnBurst_Instantaneous(Spawn Count=1, Use Spawn Probability=false)
   &SpawnRate(SpawnRate=50, Use Spawn Probability=false) {enabled=false}
  ]
  particleSpawnScript [
   &InitializeParticle(
    Lifetime Max=0.133
    Lifetime Min=0.1
    Lifetime Mode=::NewEnumerator1
    Mesh Scale=&Multiply_VectorByFloat(Vector=&Multiply_Vector(A=<1.5 1.5 1>, B=<1 1 1>), Float=$User.GlobalScale)
    Mesh Scale Mode=::NewEnumerator4
    UsePositionOffset=false
    Write Material Random=true
   )
   &InitialMeshOrientation(
    Mesh Axis To Orient=<0 0 1>
    Mesh Orientation Mode=::NewEnumerator2
    Orientation Coordinate Space=::Local
    Orientation Vector=<0 0 -1>
    Rotation=&RandomRangeVector(Minimum=<0 0 0>, Maximum=<0 0 1>)
    Use Rotation Vector=true
   )
  ]
  particleUpdateScript [
   &ParticleState(Kill Particles When Lifetime Has Elapsed=false, Loop Particles Lifetime=true)
   /SetVariables(Particles.ThrusterMask:NiagaraFloat=&Max_Float(
     A:NiagaraFloat=&FloatFromCurve(
      FloatCurve=DataInterface() {propertyValues {Curve {keys [
          {arriveTangent=9.0725, interpMode=RCIM_Cubic, leaveTangent=9.0725, tangentMode=RCTM_User, time=0, value=0}
          <at 0.1408 0.9847 RCIM_Cubic>
          {arriveTangent=-0.0405, interpMode=RCIM_Cubic, leaveTangent=-0.0405, tangentMode=RCTM_User, time=1, value=0}
         ]}}}
      CurveIndex=$Emitter.Age
     )
     B:NiagaraFloat=&Max_Float(
      A:NiagaraFloat=$System.Timeline.NormalizedTime
      B:NiagaraFloat=&MakeCustomFloatFromBool(
       Bool=&SetBoolByIntComparison(A=$Particles.PartitionParticles.Partition, B=0, Comparison Type=::NewEnumerator4)
       True Float=0
       False Float=&MakeCustomFloatFromBool(
        Bool=&SetBoolByFloatComparison(
         A=&MakeFloatFromInt(Int=$Particles.PartitionParticles.Partition)
         B=3
         Comparison Type=::NewEnumerator6
        )
        True Float=&Multiply_Float(A=$User.RightDot, B=-1)
        False Float=$User.RightDot
       )
      )
     )
    )) {enabled=false, isSetParameters=true}
   &ScaleMeshSize(Scale Factor=&Multiply_Vector(
     A=&Lerp_Vector(A=<1 1 1>, B=<1.5 1.5 2.5>, Alpha=$Particles.ThrusterMask)
     B=&Multiply_VectorByFloat(
      Vector=<1 1 1>
      Float=&FloatFromCurve(
       FloatCurve=DataInterface() {propertyValues {Curve {keys [
           {arriveTangent=9.0725, interpMode=RCIM_Cubic, leaveTangent=9.0725, tangentMode=RCTM_User, time=0, value=1}
           <at 0.1408 2 RCIM_Cubic>
           {arriveTangent=-0.0405, interpMode=RCIM_Cubic, leaveTangent=-0.0405, tangentMode=RCTM_User, time=1, value=1}
          ]}}}
       CurveIndex=$System.Timeline001.NormalizedTime
      )
     )
    ))
   &CameraOffset(Camera Offset Amount=2)
  ]
  rendererClasses [&NiagaraMeshRendererProperties]
  renderers [
   {rendererClass=&NiagaraMeshRendererProperties, rendererIndex=0}
   {rendererClass=&NiagaraMeshRendererProperties, rendererIndex=1}
  ]
 }}
summary {
 systemName=NS_BattleStage_Thruster_S7
 emitters [{emitterName=Flash, rendererClasses [&NiagaraMeshRendererProperties]}]
 userVariables [
  User.FlameScale {type=Vector3f, description="", defaultValue {struct=Vector3f, value=<1 1 1>}}
  User.GlobalScale {type=NiagaraFloat, defaultValue=2, description=""}
  User.LightSize {type=NiagaraFloat, defaultValue=1, description=""}
  User.SpawnRate {type=NiagaraFloat, defaultValue=1, description=""}
  User.SpriteFlame {type=NiagaraFloat, defaultValue=1, description=""}
 ]
}
